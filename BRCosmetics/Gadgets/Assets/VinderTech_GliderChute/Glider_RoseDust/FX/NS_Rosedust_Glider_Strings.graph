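# Niagara system drawing the ribbon strings of the RoseDust glider chute, briefly animating them out and fading them away.
# Two mirrored emitters each burst four five-particle ribbons, anchored between a root and four end bones sampled from the User.Glider skeletal mesh.
# A system timeline extends the pivots toward the end bones while a second eases noise force.
# Pendulum constraints, loosest mid-string, add sway under gravity, with width and alpha curves shaping each ribbon.
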NiagaraSystem /BRCosmetics/Gadgets/Assets/VinderTech_GliderChute/Glider_RoseDust/FX/NS_Rosedust_Glider_Strings
dependencies {
 usedDataInterfaces [&NiagaraDataInterfaceArrayFloat, &NiagaraDataInterfaceArrayInt32, &NiagaraDataInterfaceCurve]
 usedDynamicInputs [&FN_FrontEndFloatModifier]
 usedModules [
  &SystemState
  &Timeline
  &self:SystemSpawnScript.SystemScriptSource.SystemScriptGraph.NiagaraNodeAssignment_0.SetVariables_#1
 ]
}
emitters {
 Strings {
  emitterUpdateScript [
   &EmitterState()
   &SpawnBurst_Instantaneous(Spawn Count=20, Spawn Time=0.2, Use Spawn Probability=false, UseLoopCountLimit=false)
  ]
  particleSpawnScript [
   &InitializeRibbon(Lifetime=0.5, Ribbon Width=4, UsePositionOffset=false, Write Material Random=true)
   &FN_SetSpawnGroup(Inclusivity=true, Normalize Spawn Group=true, Particles Per Group=5, Pattern=::NewEnumerator0, Set Ribbon ID and Ribbon Link Order=true)
  ]
  particleUpdateScript [
   &ParticleState(Kill Particles When Lifetime Has Elapsed=false, Let Infinitely Lived Particles Die When Emitter Deactivates=true)
   &DynamicMaterialParameters(
    Index 0 Param 1=1
    Index 0 Param 2=&SelectFloatFromArray(Direct Array Index=$Particles.SpawnGroup, Float Selection Array=DataInterface() {propertyValues {FloatData [0.22,0.44,0.66,0.88]}})
    Module.Param0WriteEnabled=true
    Module.Param1WriteEnabled=true
    Module.Param2WriteEnabled=false
    Module.Param3WriteEnabled=false
   ) {module=DynamicMaterialParameters001}
   &SampleSkeletalMesh(
    AdditionalYawPitchRoll=false
    Bone Sampling Mode:ENiagaraBoneSamplingMode=::NewEnumerator4
    Sampled Orientation Coordinate Space:ENiagaraCoordinateSpace=::World
    Sampled Position Coordinate Space:ENiagaraCoordinateSpace=::World
    Sampled Velocity Coordinate Space:ENiagaraCoordinateSpace=::World
    Skeletal Mesh:NiagaraDataInterfaceSkeletalMesh=$User.Glider
   )
   &SampleSkeletalMesh(
    AdditionalYawPitchRoll=false
    Bone / Socket Index:NiagaraInt32=&SelectIntFromArray(
     Direct Array Index=&SpawnGroup()
     Direct Array Mode=::NewEnumerator1
     Int Selection Array=DataInterface() {propertyValues {IntData [1,2,3,4]}}
    )
    Bone Sampling Mode:ENiagaraBoneSamplingMode=::NewEnumerator4
    Position Offset Coordinate Space:ENiagaraCoordinateSpace=::World
    Sampled Orientation Coordinate Space:ENiagaraCoordinateSpace=::World
    Sampled Position Coordinate Space:ENiagaraCoordinateSpace=::World
    Sampled Velocity Coordinate Space:ENiagaraCoordinateSpace=::World
    Skeletal Mesh:NiagaraDataInterfaceSkeletalMesh=$User.Glider
   ) {module=SampleSkeletalMesh001}
   /SetVariables(Particles.PositionOrigin:Vector3f=&Lerp_Vector(
     A=$Particles.SampleSkeletalMesh.SampledPosition
     B=&Lerp_Vector(
      A=$Particles.SampleSkeletalMesh.SampledPosition
      B=&Subtract_Vector(A=$Particles.SampleSkeletalMesh001.SampledPosition, B=<0 0 0>)
      Alpha=$System.Timeline.NormalizedTime
     )
     Alpha=$Particles.Normalized Spawn Group
    )) {isSetParameters=true, module=SetVariables_#2}
   &GravityForce()
   &VectorNoiseForce(
    Force Amount=&Lerp_Float(A=15000, B=1500, Alpha=$System.Timeline001.NormalizedTime)
    Module.Use Random Noise Time Interval=false
   )
   &PendulumSetup(
    Module.Enable Spring Driver=false
    Pendulum Length:NiagaraFloat=&FloatFromCurve(
     FloatCurve=DataInterface() {propertyValues {Curve {keys [
         {arriveTangent=3.8991, interpMode=RCIM_Cubic, leaveTangent=3.8991, leaveTangentWeight=0.2364, tangentMode=RCTM_User, time=0, value=0}
         <at 0.5 1 RCIM_Cubic>
         {arriveTangent=-3.8444, arriveTangentWeight=0.4978, interpMode=RCIM_Cubic, leaveTangent=-3.8444, tangentMode=RCTM_User, time=1, value=0}
        ]}}}
     CurveIndex=$Particles.Normalized Spawn Group
     Scale Curve=20
    )
    Pendulum Pivot:Vector3f=$Particles.PositionOrigin
    Rigid Pendulum=false
   )
   &SolveForcesAndVelocity(Clamp Velocity=false, Limit Acceleration=false)
   &PendulumConstraint(Calculate and Output Potential Energy=true)
   &CalculateAccurateVelocity()
   &ScaleRibbonWidth(Ribbon Width Scale=&FloatFromCurve(
     FloatCurve=DataInterface() {propertyValues {Curve {keys [
         {arriveTangent=1.6505, interpMode=RCIM_Cubic, leaveTangent=1.6505, leaveTangentWeight=0.3404, tangentMode=RCTM_User, time=0, value=0.5}
         {arriveTangent=-0.25, interpMode=RCIM_Cubic, leaveTangent=-0.25, time=0.75, value=1.5}
         {arriveTangent=-11.5328, interpMode=RCIM_Cubic, leaveTangent=-11.5328, tangentMode=RCTM_User, time=1, value=0.25}
        ]}}}
     CurveIndex=&Frac_Float(Float=$Particles.Normalized Spawn Group)
     Scale Curve=&FloatFromCurve(
      FloatCurve=DataInterface() {propertyValues {Curve {keys [
          {tangentMode=RCTM_User, time=0.05, value=0}
          {tangentMode=RCTM_User, time=0.1, value=1}
          {interpMode=RCIM_Cubic, tangentMode=RCTM_User, time=0.5, value=1}
          {arriveTangent=-1.9649, interpMode=RCIM_Cubic, leaveTangent=-1.9649, tangentMode=RCTM_User, time=1, value=0}
         ]}}}
      CurveIndex=$Emitter.Age
     )
    ))
   &ScaleColor(
    Scale Alpha=&FloatFromCurve(FloatCurve=DataInterface() {propertyValues {Curve {keys [
         {arriveTangent=-0.4049, interpMode=RCIM_Cubic, leaveTangent=-0.4049, tangentMode=RCTM_User, time=0.5, value=1}
         {arriveTangent=-2.8515, interpMode=RCIM_Cubic, leaveTangent=-2.8515, tangentMode=RCTM_User, time=1, value=0}
        ]}}})
    ScaleA=true
    ScaleRGB=true
   )
  ]
  rendererClasses [&NiagaraRibbonRendererProperties]
  renderers [{rendererClass=&NiagaraRibbonRendererProperties}]
 }
 Strings001 {
  emitterUpdateScript [
   &EmitterState()
   &SpawnBurst_Instantaneous(Spawn Count=20, Use Spawn Probability=false, UseLoopCountLimit=false)
  ]
  particleSpawnScript [
   &InitializeRibbon(Lifetime=0.5, Ribbon Width=4, UsePositionOffset=false, Write Material Random=true)
   &FN_SetSpawnGroup(Inclusivity=true, Normalize Spawn Group=true, Particles Per Group=5, Pattern=::NewEnumerator0, Set Ribbon ID and Ribbon Link Order=true)
  ]
  particleUpdateScript [
   &ParticleState(Kill Particles When Lifetime Has Elapsed=false, Let Infinitely Lived Particles Die When Emitter Deactivates=true)
   &DynamicMaterialParameters(
    Index 0 Param 1=1
    Index 0 Param 2=&SelectFloatFromArray(
     Direct Array Index=&SpawnGroup()
     Float Selection Array=DataInterface() {propertyValues {FloatData [0.44,0.66,0.88,0.22]}}
    )
    Module.Param0WriteEnabled=true
    Module.Param1WriteEnabled=true
    Module.Param2WriteEnabled=false
    Module.Param3WriteEnabled=false
   ) {module=DynamicMaterialParameters001}
   &SampleSkeletalMesh(
    AdditionalYawPitchRoll=false
    Bone / Socket Index=5
    Bone Sampling Mode:ENiagaraBoneSamplingMode=::NewEnumerator4
    Sampled Orientation Coordinate Space:ENiagaraCoordinateSpace=::World
    Sampled Position Coordinate Space:ENiagaraCoordinateSpace=::World
    Sampled Velocity Coordinate Space:ENiagaraCoordinateSpace=::World
    Skeletal Mesh:NiagaraDataInterfaceSkeletalMesh=$User.Glider
   )
   &SampleSkeletalMesh(
    AdditionalYawPitchRoll=false
    Bone / Socket Index:NiagaraInt32=&SelectIntFromArray(Direct Array Index=$Particles.SpawnGroup, Int Selection Array=DataInterface() {propertyValues {IntData [6,7,8,9]}})
    Bone Sampling Mode:ENiagaraBoneSamplingMode=::NewEnumerator4
    Position Offset Coordinate Space:ENiagaraCoordinateSpace=::World
    Sampled Orientation Coordinate Space:ENiagaraCoordinateSpace=::World
    Sampled Position Coordinate Space:ENiagaraCoordinateSpace=::World
    Sampled Velocity Coordinate Space:ENiagaraCoordinateSpace=::World
    Skeletal Mesh:NiagaraDataInterfaceSkeletalMesh=$User.Glider
   ) {module=SampleSkeletalMesh001}
   /SetVariables(Particles.PositionOrigin:Vector3f=&Lerp_Vector(
     A=$Particles.SampleSkeletalMesh.SampledPosition
     B=&Lerp_Vector(
      A=$Particles.SampleSkeletalMesh.SampledPosition
      B=&Subtract_Vector(A=$Particles.SampleSkeletalMesh001.SampledPosition, B=<0 0 0>)
      Alpha=$System.Timeline.NormalizedTime
     )
     Alpha=$Particles.Normalized Spawn Group
    )) {isSetParameters=true, module=SetVariables_#2}
   &GravityForce()
   &VectorNoiseForce(
    Force Amount=&Lerp_Float(A=15000, B=1500, Alpha=$System.Timeline001.NormalizedTime)
    Module.Use Random Noise Time Interval=false
   )
   &PendulumSetup(
    Module.Enable Spring Driver=false
    Pendulum Length:NiagaraFloat=&FloatFromCurve(
     FloatCurve=DataInterface() {propertyValues {Curve {keys [
         {arriveTangent=3.8991, interpMode=RCIM_Cubic, leaveTangent=3.8991, leaveTangentWeight=0.2364, tangentMode=RCTM_User, time=0, value=0}
         <at 0.5 1 RCIM_Cubic>
         {arriveTangent=-3.8444, arriveTangentWeight=0.4978, interpMode=RCIM_Cubic, leaveTangent=-3.8444, tangentMode=RCTM_User, time=1, value=0}
        ]}}}
     CurveIndex=$Particles.Normalized Spawn Group
     Scale Curve=20
    )
    Pendulum Pivot:Vector3f=$Particles.PositionOrigin
    Rigid Pendulum=false
   )
   &SolveForcesAndVelocity(Clamp Velocity=false, Limit Acceleration=false)
   &PendulumConstraint(Calculate and Output Potential Energy=true)
   &CalculateAccurateVelocity()
   &ScaleRibbonWidth(Ribbon Width Scale=&FloatFromCurve(
     FloatCurve=DataInterface() {propertyValues {Curve {keys [
         {arriveTangent=1.6505, interpMode=RCIM_Cubic, leaveTangent=1.6505, leaveTangentWeight=0.3404, tangentMode=RCTM_User, time=0, value=0.5}
         {arriveTangent=-0.25, interpMode=RCIM_Cubic, leaveTangent=-0.25, time=0.75, value=1.5}
         {arriveTangent=-11.5328, interpMode=RCIM_Cubic, leaveTangent=-11.5328, tangentMode=RCTM_User, time=1, value=0.25}
        ]}}}
     CurveIndex=&Frac_Float(Float=$Particles.Normalized Spawn Group)
     Scale Curve=&FloatFromCurve(
      FloatCurve=DataInterface() {propertyValues {Curve {keys [
          {tangentMode=RCTM_User, time=0.05, value=0}
          {tangentMode=RCTM_User, time=0.1, value=1}
          {interpMode=RCIM_Cubic, tangentMode=RCTM_User, time=0.5, value=1}
          {arriveTangent=-1.9649, interpMode=RCIM_Cubic, leaveTangent=-1.9649, tangentMode=RCTM_User, time=1, value=0}
         ]}}}
      CurveIndex=$Emitter.Age
     )
    ))
   &ScaleColor(
    Scale Alpha=&FloatFromCurve(FloatCurve=DataInterface() {propertyValues {Curve {keys [
         {arriveTangent=-0.4049, interpMode=RCIM_Cubic, leaveTangent=-0.4049, tangentMode=RCTM_User, time=0.5, value=1}
         {arriveTangent=-2.8515, interpMode=RCIM_Cubic, leaveTangent=-2.8515, tangentMode=RCTM_User, time=1, value=0}
        ]}}})
    ScaleA=true
    ScaleRGB=true
   )
  ]
  rendererClasses [&NiagaraRibbonRendererProperties]
  renderers [{rendererClass=&NiagaraRibbonRendererProperties}]
 }
}
summary {
 systemName=NS_Rosedust_Glider_Strings
 emitters [
  {emitterName=Strings, rendererClasses [&NiagaraRibbonRendererProperties]}
  {emitterName=Strings001, rendererClasses [&NiagaraRibbonRendererProperties]}
 ]
 userVariables [
  User.bIsFrontEnd {type=NiagaraBool}
  User.bIsFrontEndPreview {type=NiagaraBool}
  User.Glider {
   type=NiagaraDataInterfaceSkeletalMesh
   defaultValue {
    struct=NiagaraExt_VariableValue_DataInterface
    value {dataInterface=&self:NiagaraDataInterfaceSkeletalMesh_5}
   }
  }
 ]
}
