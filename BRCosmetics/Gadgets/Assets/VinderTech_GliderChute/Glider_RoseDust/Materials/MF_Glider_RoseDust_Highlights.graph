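# Material function for the RoseDust glider that adds stylized highlight dots and shadow hatching to incoming material attributes, toggled by the UseRoseDustHighlights static switch.
# Procedural patterns use a camera-facing virtual plane anchored near the actor, with FOV- and distance-adjusted density, and are thresholded by lighting gradients.
# Their colors pass through a LUT into emissive, blended by a combined mask.
# It also builds Fresnel-driven thin-film emissive and sets subsurface shading with Fresnel-based opacity.
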MaterialFunction /BRCosmetics/Gadgets/Assets/VinderTech_GliderChute/Glider_RoseDust/Materials/MF_Glider_RoseDust_Highlights
graph {nodes [
  FunctionOutput(None=StaticSwitchParameter(True=NamedRerouteUsage(), False=NamedRerouteUsage()) {group=1_RoseDust, parameterName=UseRoseDustHighlights, sortPriority=0})
  n1 = TextureSample() {samplerType=SAMPLERTYPE_Masks, texture=/BRCosmetics/Gadgets/Assets/VinderTech_GliderChute/Glider_RoseDust/Textures/T_Glider_RoseDust_S}
  n2 = TextureSample() {samplerType=SAMPLERTYPE_Normal, texture=/BRCosmetics/Gadgets/Assets/VinderTech_GliderChute/Glider_RoseDust/Textures/T_Glider_RoseDust_N}
  n3 = SetMaterialAttributes!(
   MaterialAttributes
   BaseColor=TextureSample() {texture=/BRCosmetics/Gadgets/Assets/VinderTech_GliderChute/Glider_RoseDust/Textures/T_Glider_RoseDust_D}.RGB
   Specular=@n1.R
   Metallic=@n1.R
   Roughness=@n1.R
   Normal=@n2.RGB
  ) {attributeSetTypes [
    #1
    #2
    #3
    #4
    #5
   ]}
  NamedRerouteDeclaration(None=GetMaterialAttributes(None=FunctionInput(Preview=@n3) {inputName=MA_PreToon, inputType=FunctionInput_MaterialAttributes, sortPriority=3}) {attributeGetTypes [#1]}.BaseColor) {name=BaseColor, nodeColor=<rgba 0.5937 1 0 1>}
  n4 = GetMaterialAttributes(None=NamedRerouteUsage()) {attributeGetTypes [#6]}
  n5 = ScalarParameter() {group=ThinFilm, parameterName=ThinFilm_Offset, sliderMax=1, sortPriority=11}
  n6 = ScalarParameter() {defaultValue=1, group=ThinFilm, parameterName=ThinFilm_Scale, sliderMax=1, sortPriority=11}
  n7 = Fresnel(ExponentIn=ScalarParameter() {defaultValue=0.5, group=ThinFilm, parameterName=ThinFilm_Warp, sliderMax=1, sortPriority=11}) {baseReflectFraction=0}
  n8 = TextureSampleParameter2D(UVs=Add(A=@n5, B=Multiply(A=@n6, B=@n7))) {group=ThinFilm, parameterName=Thin Film Texture, sortPriority=2, texture=/Game/Global/Textures/Default/Thinfilm/T_ThinFilm_Spectrum_COLOR}
  n9 = ScalarParameter() {defaultValue=0.1, group=ThinFilm, parameterName=ThinFilm_Intensity_Webbing, sliderMax=1, sortPriority=4}
  n10 = NamedRerouteUsage()
  n11 = NamedRerouteUsage()
  n12 = Multiply(A=OneMinus(None=Ceil(None=@n10)), B=@n11)
  n13 = Multiply(
   A=@n8.RGB
   B=LinearInterpolate(
    A=ScalarParameter() {defaultValue=1, group=ThinFilm, parameterName=ThinFilm_Intensity_NegativeSpace, sliderMax=1, sortPriority=4}
    B=Multiply(A=@n9, B=@n12)
    Alpha=@n11
   )
  )
  NamedRerouteDeclaration(None=SetMaterialAttributes!(MaterialAttributes=@n4.MaterialAttributes, EmissiveColor=Add(A=@n4.EmissiveColor, B=@n13)) {attributeSetTypes [#6]}) {name=ThinFilm, nodeColor=<rgba 1 0.3047 0 1>}
  n14 = Multiply(A=NamedRerouteUsage()) {constB=-1}
  n15 = ScalarParameter() {group=1_RoseDust, parameterName=ShadowHatchingOffset, sliderMax=0.5, sliderMin=-0.5, sortPriority=20}
  n16 = Subtract(A=Multiply(A=Add(A=@n14, B=@n15) {constB=-0.25}) {constB=0.5, desc=:/}, B=NamedRerouteUsage())
  n17 = NamedRerouteDeclaration(None=SmoothStep(Min=Constant(), Max=Constant() {r=0.1}, Value=Saturate(None=@n16))) {name=ShadowHatchingMask, nodeColor=<rgba 0.3594 0.3594 0.3594 1>}
  n18 = VectorParameter() {group=1_RoseDust, parameterName=ShadowHatchingColor, sortPriority=10, defaultValue=<rgba 0.18 0.18 0.18 0.5>}
  NamedRerouteDeclaration(None=AppendVector(A=Multiply(A=@n17, B=@n18.RGB), B=@n18.A)) {name=ShadowHatching, nodeColor=<rgba 0.099 0.099 0.099 1>}
  n19 = TextureSample() {samplerType=SAMPLERTYPE_Masks, texture=/BRCosmetics/Gadgets/Assets/VinderTech_GliderChute/Glider_RoseDust/Textures/T_Glider_RoseDust_FX}
  NamedRerouteDeclaration(None=ChannelMaskParameter(None=FunctionInput(Preview=@n19.RGBA) {inputName=FX, inputType=FunctionInput_Vector4, sortPriority=1}) {group=1_RoseDust, parameterName=FX_Channel}.R) {name=WebbingMask, nodeColor=<rgba 0 1 0.5313 1>}
  n20 = MaterialFunctionCall() {materialFunction=&CameraDirectionVector}
  n21 = Multiply(A=Constant3Vector() {constant=<rgba 1 1 0 0>}, B=@n20.Result)
  n22 = Add(A=Multiply(B=Normalize(VectorInput=@n21)) {constA=0.25}, B=@n20.Result)
  n23 = Multiply(A=@n20.Result) {constB=50}
  n24 = Add(A=ActorPositionWS(), B=Constant3Vector() {constant=<rgba 0 0 90 1>})
  n25 = Constant() {r=64}
  n26 = Constant() {r=128}
  n27 = InverseLinearInterpolate(
   A=Constant() {r=45}
   B=Constant() {r=90}
   Value=MaterialFunctionCall() {materialFunction=/Engine/Functions/Engine_MaterialFunctions02/Utility/FOV}.FOV angle in degrees
  )
  NamedRerouteDeclaration(None=MaterialFunctionCall(Plane Normal Axis=Normalize(VectorInput=@n22), Plane Center=Add(A=@n23, B=@n24), UV Scale=LinearInterpolate(A=@n25, B=@n26, Alpha=@n27)) {materialFunction=&VirtualPlaneCoordinates}.UVs) {name=ObjectAlighedScreenPosition, nodeColor=<rgba 1 0.867 0 1>}
  n28 = Custom() {code="return View.DirectionalLightDirection;", description=DirectionalLightDirection, inputs []}
  n29 = NamedRerouteUsage()
  n30 = Multiply(A=NamedRerouteDeclaration(None=DotProduct(A=@n28, B=@n29)) {name=LdotN, nodeColor=<rgba 0.9219 1 0 1>}) {constB=0.5}
  NamedRerouteDeclaration(None=Saturate(None=Add(A=@n30) {constB=0.5})) {name=Lambert, nodeColor=<rgba 0 0.9297 1 1>}
  n31 = Subtract(A=Distance(A=ActorPositionWS(), B=CameraPositionWS())) {constB=100}
  NamedRerouteDeclaration(None=Saturate(None=Divide(A=@n31) {constB=400})) {name=DistanceBlending, nodeColor=<rgba 0.4766 1 0 1>}
  n32 = ScalarParameter() {defaultValue=70, group=1_RoseDust, parameterName=HatchingAmount, sliderMax=512, sortPriority=20}
  n33 = Multiply(A=NamedRerouteUsage(), B=LinearInterpolate(A=@n32, B=Multiply(A=@n32) {constB=0.5}, Alpha=NamedRerouteUsage()))
  n34 = Constant() {r=1}
  n35 = Frac(None=ComponentMask(None=Rotator(Coordinate=@n33, Time=@n34) {speed=0.75}) {g=true})
  NamedRerouteDeclaration(None=Abs(None=Add(A=@n35) {constB=-0.5})) {name=HatchingGradient, nodeColor=<rgba 0.3042 0 1 1>}
  n36 = ScalarParameter() {defaultValue=70, group=1_RoseDust, parameterName=DotsAmount, sliderMax=512, sortPriority=20}
  n37 = Multiply(A=NamedRerouteUsage(), B=LinearInterpolate(A=@n36, B=Multiply(A=@n36) {constB=0.5}, Alpha=NamedRerouteUsage()))
  n38 = Constant() {r=1}
  n39 = Add(A=Frac(None=Rotator(Coordinate=@n37, Time=@n38) {speed=0.75})) {constB=-0.5}
  n40 = Constant()
  NamedRerouteDeclaration(None=Add(A=Distance(A=@n39, B=@n40)) {constB=-0.1, desc=to get full transition}) {name=DotsGradient, nodeColor=<rgba 1 0.0938 0 1>}
  n41 = NamedRerouteUsage()
  n42 = ScalarParameter() {defaultValue=-0.3, group=1_RoseDust, parameterName=HighlightDotsOffset, sliderMax=0.5, sliderMin=-0.5, sortPriority=20}
  n43 = Subtract(
   A=Multiply(
    A=Add(A=@n41, B=@n42) {constB=-0.25}
    B=ScalarParameter() {defaultValue=1.5, group=1_RoseDust, parameterName=HighlightDotsTransition, sliderMax=2, sortPriority=20}
   ) {constB=1.5}
   B=NamedRerouteUsage()
  )
  n44 = NamedRerouteDeclaration(None=SmoothStep(Min=Constant(), Max=Constant() {r=0.1}, Value=Saturate(None=@n43))) {name=HighlightDotsMask, nodeColor=<rgba 0.4766 1 0 1>}
  n45 = VectorParameter() {group=1_RoseDust, parameterName=HighlightDotsColor, sortPriority=10, defaultValue=<rgba 0.9 0.9 0.9 1>}
  NamedRerouteDeclaration(None=AppendVector(A=Multiply(A=@n44, B=@n45.RGB), B=@n45.A)) {name=HighlightDots, nodeColor=<rgba 1 0 0.6797 1>}
  n46 = VectorParameter() {group=1_RoseDust, parameterName=WhiteSuitShadowHatchingColor, sortPriority=10, defaultValue=<rgba 0.18 0.18 0.18 0.5>}
  NamedRerouteDeclaration(None=AppendVector(A=Multiply(A=@n17, B=@n46.RGB), B=@n46.A)) {name=WhiteSuitShadowHatching, nodeColor=<rgba 0.8594 0.8594 0.8594 1>}
  NamedRerouteDeclaration(None=FunctionInput(Preview=@n3) {inputName=MA, inputType=FunctionInput_MaterialAttributes}) {name=MA, nodeColor=<rgba 0.3594 1 0 1>}
  n47 = NamedRerouteUsage()
  n48 = NamedRerouteUsage()
  n49 = NamedRerouteUsage()
  n50 = MaterialFunctionCall!(In=Add(A=ComponentMask(None=@n48) {b=true, g=true, r=true}, B=ComponentMask(None=@n49) {b=true, g=true, r=true}), LUT, LUT Multiplier) {materialFunction=/Game/Packages/Fortress_SharedMaterials/Anime_Master_Material/MF_ApplyLUT}
  n51 = Add(
   A=Multiply(A=NamedRerouteUsage(), B=ComponentMask(None=@n48) {a=true})
   B=Multiply(A=NamedRerouteUsage(), B=ComponentMask(None=@n49) {a=true})
  )
  n52 = LinearInterpolate(
   A=ScalarParameter() {defaultValue=1, group=1_RoseDust, parameterName=SubSurface_Opacity_Inner, sliderMax=1}
   B=ScalarParameter() {group=1_RoseDust, parameterName=SubSurface_Opacity_Edge, sliderMax=1}
   Alpha=Saturate(None=Fresnel() {baseReflectFraction=0, exponent=1})
  )
  n53 = SetMaterialAttributes!(
   MaterialAttributes=BlendMaterialAttributes(
    A=NamedRerouteUsage()
    B=SetMaterialAttributes!(MaterialAttributes=@n47, EmissiveColor=@n50.Result) {attributeSetTypes [#6]}
    Alpha=NamedRerouteDeclaration(None=@n51) {name=DotsAndHatchingMask, nodeColor=<rgba 1 0.6094 0 1>}
   )
   Metallic=Constant()
   ShadingModel=ShadingModel() {shadingModel=MSM_Subsurface}
   SubsurfaceColor=Multiply(A=VectorParameter() {group=1_RoseDust, parameterName=SubSurface_Color, defaultValue=<rgba 0 0.6897 1 1>}.RGB, B=NamedRerouteUsage())
   Opacity=OneMinus(None=Saturate(None=@n52))
  ) {attributeSetTypes [
    #3
    #7
    #8
    #9
   ]}
  NamedRerouteDeclaration(None=@n53) {name=MA_Out, nodeColor=<rgba 1 0.0703 0 1>}
  n54 = Saturate(None=DotProduct(A=CameraVectorWS(), B=PixelNormalWS()))
  n55 = ScalarParameter() {defaultValue=20, group=1_RoseDust, parameterName=SpecPower, sliderMax=32, sortPriority=20}
  NamedRerouteDeclaration(None=Multiply(
    A=Power(Base=@n54, Exp=@n55)
    B=ScalarParameter() {defaultValue=0.1, group=1_RoseDust, parameterName=SpecularIntensity, sliderMax=2, sortPriority=20}
   )) {name=Specular, nodeColor=<rgba 0 1 1 1>}
  NamedRerouteDeclaration(None=Transform(None=@n2.RGB)) {name=WSNormals, nodeColor=<rgba 0 0.9531 1 1>}
  PixelNormalWS()
  n56 = ScalarParameter() {defaultValue=0.5, group=1_RoseDust, parameterName=WebTransparency, sliderMax=1}
  n57 = NamedRerouteUsage()
  SetMaterialAttributes!(
   MaterialAttributes=@n53
   OpacityMask=MaterialFunctionCall(Alpha Threshold=LinearInterpolate(A=@n56, Alpha=@n57) {constA=1}) {materialFunction=&DitherTemporalAA}.Result
  ) {attributeSetTypes [#10]}
  StaticSwitchParameter(
   True=MaterialFunctionCall!(
    Time=ScalarParameter() {defaultValue=1, group=ThinFilm, parameterName=ThinFilm_Animated_TimeDivisor, sliderMax=1, sortPriority=10}
    Divisor
    Mod=ScalarParameter() {defaultValue=1, group=ThinFilm, parameterName=ThinFilm_Animated_TimeModulo, sliderMax=1, sortPriority=10}
   ) {materialFunction=/Game/Effects/Fort_Effects/Materials/_Functions/MF_ModuloTime}.Unclamped
   False=@n5
  ) {group=ThinFilm, parameterName=AnimateThinFilm, sortPriority=1}
 ]}
