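# Character-style material for the RoseDust glider, built for skeletal meshes with morph targets and mesh deformers.
# Material attributes pass from the shared character PreFX function through anime toon shading into a RoseDust-specific highlights function, which also receives the pre-toon result.
# Character PostFX and cinematic PostFX functions finish the chain.
# Shading model comes from the graph, spanning lit, subsurface, profile and cloth, with a borrowed Ice subsurface profile.
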Material /BRCosmetics/Gadgets/Assets/VinderTech_GliderChute/Glider_RoseDust/Materials/M_Glider_RoseDust
graph {
 properties {
  bAutomaticallySetUsageInEditor=false
  bUseMaterialAttributes=true
  bUsedWithMeshDeformer=true
  bUsedWithMorphTargets=true
  bUsedWithSkeletalMesh=true
  pixelDepthOffsetMode=PDOM_Legacy
  shadingModel=MSM_FromMaterialExpression
  subsurfaceProfile=/Game/Accessories/FORT_Backpacks/M_MED_Heist/Textures/Ice_SSP
  usedShadingModels=Default Lit | Subsurface | Subsurface Profile | Cloth
 }
 nodes [
  n1 = MaterialFunctionCall() {materialFunction=&BaseFN_CharacterFunction_PreFX}
  n2 = MaterialFunctionCall!(
   Material=@n1.Result
   LitDiffuseOverride
   SSC_AlphaOverride
   Tint Mask
   DF InkLine Visibility Mask
   DF Inkline Injection
   DF Inkline Color Override
   DF Inkline Color Override Mask
   CustomUVs
  ) {materialFunction=/Game/Packages/Fortress_SharedMaterials/Anime_Master_Material/MF_AnimeShading}
  n3 = MaterialFunctionCall(
   AO=@n1.AO
   Distance Blend=@n1.AO
   Fresnel=@n1.AO
   Gmap=@n1.AO
   Pre Skinned Local Position=@n1.AO
   Material=MaterialFunctionCall!(MA=@n2.Result, FX=@n2.Result, MA_PreToon=@n1.Result) {materialFunction=&/MF_Glider_RoseDust_Highlights}.Result
  ) {materialFunction=&BaseFN_CharacterFunction_PostFX}
  MaterialFunctionCall(In=@n3.Result) {materialFunction=&CINE_MF_CharPostFX}
 ]
}
